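# Simplified physics asset for the large male Constructor player character, apparently serving as its capsule shadow volume, judging by the name and capsule-only shapes.
# One capsule each covers pelvis, upper spine, upper and lower arms, thighs, calves and feet, with no bodies for head or hands.
# Constraints chain limbs to spine or pelvis with all swing and twist motion free, so the angular limit values do not restrict posing.
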PhysicsAsset /Game/Characters/Player/Male/Large/ShadowVolume/SK_M_LRG_Constructor_01_ShadowVolume
bodies {
 calf_l [{shapeType=Capsule, length=36.025, radius=7.5852, center=<-20.225 0.2462 -0.3211>, rotation=<rot 12.8805 89.6448 -87.1583>}]
 calf_r [{shapeType=Capsule, length=33.0253, radius=8.1869, center=<21.3519 -1.5265 0.6049>, rotation=<rot 29.8449 -90.1411 94.4069>}]
 foot_l [{shapeType=Capsule, length=17.7908, radius=7.0075, center=<-3.0759 7.8574 0.8018>, rotation=<rot 1.1116 -179.3378 -87.5565>}]
 foot_r [{shapeType=Capsule, length=14.0208, radius=8.4804, center=<3.6173 -9.0329 -0.8624>, rotation=<rot -0.9745 3.8372 -92.5014>}]
 lowerarm_l [{shapeType=Capsule, length=35.005, radius=5.005, center=<16.7624 0 -0.9791>, rotation=<rot 0 -90 92.5115>}]
 lowerarm_r [{shapeType=Capsule, length=31.255, radius=7.755, center=<-14.9089 -0.0004 1.2177>, rotation=<rot 0 -89.9999 95.1651>}]
 pelvis [{shapeType=Capsule, length=16.4909, radius=17.08, center=<-8.6952 -0.5448 0.212>, rotation=<rot 0 7.9543 -1.2349>}]
 spine_03 [{shapeType=Capsule, length=37.9981, radius=22.5343, center=<-0.5611 -4.2405 -0.0079>, rotation=<rot 0 90 -90>}]
 thigh_l [{shapeType=Capsule, length=37.5694, radius=8.564, center=<-23.5737 0.5519 1.9716>, rotation=<rot -3.3627 91.7452 -83.4788>}]
 thigh_r [{shapeType=Capsule, length=39.6273, radius=7.9058, center=<22.1628 -1.6318 -0.7517>, rotation=<rot -1.9569 -93.7061 -90.9145>}]
 upperarm_l [{shapeType=Capsule, length=19.553, radius=10.665, center=<14.6577 -2.14 -0.4415>, rotation=<rot 18.6565 88.4767 -96.7952>}]
 upperarm_r [{shapeType=Capsule, length=16.6065, radius=10.4698, center=<-10.9911 -3.2299 1.351>, rotation=<rot 51.5557 -93.2981 -91.6626>}]
}
constraints [
 {
  bone1Name=upperarm_l
  bone2Name=spine_03
  swing1LimitDegrees=45
  swing1Motion=Free
  swing2LimitDegrees=45
  swing2Motion=Free
  twistLimitDegrees=45
  twistMotion=Free
 }
 {
  bone1Name=lowerarm_l
  bone2Name=upperarm_l
  swing1LimitDegrees=45
  swing1Motion=Free
  swing2LimitDegrees=45
  swing2Motion=Free
  twistLimitDegrees=45
  twistMotion=Free
 }
 {
  bone1Name=lowerarm_r
  bone2Name=upperarm_r
  swing1LimitDegrees=45
  swing1Motion=Free
  swing2LimitDegrees=45
  swing2Motion=Free
  twistLimitDegrees=45
  twistMotion=Free
 }
 {
  bone1Name=upperarm_r
  bone2Name=spine_03
  swing1LimitDegrees=45
  swing1Motion=Free
  swing2LimitDegrees=45
  swing2Motion=Free
  twistLimitDegrees=45
  twistMotion=Free
 }
 {
  bone1Name=calf_l
  bone2Name=thigh_l
  swing1LimitDegrees=45
  swing1Motion=Free
  swing2LimitDegrees=45
  swing2Motion=Free
  twistLimitDegrees=45
  twistMotion=Free
 }
 {
  bone1Name=foot_l
  bone2Name=calf_l
  swing1LimitDegrees=45
  swing1Motion=Free
  swing2LimitDegrees=45
  swing2Motion=Free
  twistLimitDegrees=45
  twistMotion=Free
 }
 {
  bone1Name=foot_r
  bone2Name=calf_r
  swing1LimitDegrees=45
  swing1Motion=Free
  swing2LimitDegrees=45
  swing2Motion=Free
  twistLimitDegrees=45
  twistMotion=Free
 }
 {
  bone1Name=calf_r
  bone2Name=thigh_r
  swing1LimitDegrees=45
  swing1Motion=Free
  swing2LimitDegrees=45
  swing2Motion=Free
  twistLimitDegrees=45
  twistMotion=Free
 }
 {
  bone1Name=spine_03
  bone2Name=pelvis
  swing1LimitDegrees=45
  swing1Motion=Free
  swing2LimitDegrees=45
  swing2Motion=Free
  twistLimitDegrees=45
  twistMotion=Free
 }
 {
  bone1Name=thigh_l
  bone2Name=pelvis
  swing1LimitDegrees=45
  swing1Motion=Free
  swing2LimitDegrees=45
  swing2Motion=Free
  twistLimitDegrees=45
  twistMotion=Free
 }
 {
  bone1Name=thigh_r
  bone2Name=pelvis
  swing1LimitDegrees=45
  swing1Motion=Free
  swing2LimitDegrees=45
  swing2Motion=Free
  twistLimitDegrees=45
  twistMotion=Free
 }
]
